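# Physics asset for the front-end variant of the ProdigyFire female medium cape accessory, providing capsule bodies and joints for the cape and its furry-ball ornaments.
# Center, left and right cape chains each hang four segments from a master bone, with swing limits generally widening toward the hem and little or no twist.
# Three furry-ball bases attach to the brooch with free rotation, each carrying a two-segment chain with tighter swing and twist limits.
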PhysicsAsset /BRCosmetics/Accessories/FORT_Capes/Cape_F_MED_ProdigyFire/Meshes/F_MED_ProdigyFire_Cape_Physics_FrontEnd
bodies {
 base_furryball_down [{shapeType=Capsule, length=0.505, radius=0.505, shapeName=base_furryball_down_capsule, rotation=<rot 0 -90 0>}]
 base_furryball_middle [{shapeType=Capsule, length=0.505, radius=0.505, shapeName=base_furryball_middle_capsule, rotation=<rot 0 -90 0>}]
 base_furryball_top [{shapeType=Capsule, length=0.505, radius=0.505, shapeName=base_furryball_top_capsule, rotation=<rot 0 -90 0>}]
 dyn_cape_01 [{shapeType=Capsule, length=11.7556, radius=5.9773, shapeName=dyn_cape_01_capsule, center=<7.15 0 8>, rotation=<rot 90 0 0>}]
 dyn_cape_02 [{shapeType=Capsule, length=9.602, radius=8.1308, shapeName=dyn_cape_01_capsule, center=<7.15 0 5>, rotation=<rot 90 0 0>}]
 dyn_cape_03 [{shapeType=Capsule, length=9.602, radius=8.1308, shapeName=dyn_cape_01_capsule, center=<7.15 0 5>, rotation=<rot 90 0 0>}]
 dyn_cape_04 [{shapeType=Capsule, length=10, radius=8, shapeName=dyn_cape_01_capsule, center=<10 0 5>, rotation=<rot 90 0 0>}]
 dyn_cape_MSTR [{shapeType=Capsule, length=0.505, radius=0.505, shapeName=dyn_cape_r_MSTR_capsule, rotation=<rot 0 -90 0>}]
 dyn_cape_brooch [{shapeType=Capsule, length=4, radius=4}]
 dyn_cape_l_01 [{shapeType=Capsule, length=11.7556, radius=5.9773, shapeName=dyn_cape_01_capsule, center=<7.15 0 8>, rotation=<rot 90 0 0>}]
 dyn_cape_l_02 [{shapeType=Capsule, length=9.602, radius=8.1308, shapeName=dyn_cape_01_capsule, center=<7.15 0 5>, rotation=<rot 90 0 0>}]
 dyn_cape_l_03 [{shapeType=Capsule, length=9.602, radius=8.1308, shapeName=dyn_cape_01_capsule, center=<7.15 0 5>, rotation=<rot 90 0 0>}]
 dyn_cape_l_04 [{shapeType=Capsule, length=10, radius=8, shapeName=dyn_cape_01_capsule, center=<10 0 5>, rotation=<rot 90 0 0>}]
 dyn_cape_l_MSTR [{shapeType=Capsule, length=0.505, radius=0.505, shapeName=dyn_cape_r_MSTR_capsule, rotation=<rot 0 -90 0>}]
 dyn_cape_r_01 [{shapeType=Capsule, length=11.7556, radius=5.9773, shapeName=dyn_cape_01_capsule, center=<7.15 0 8>, rotation=<rot 90 0 0>}]
 dyn_cape_r_02 [{shapeType=Capsule, length=9.602, radius=8.1308, shapeName=dyn_cape_01_capsule, center=<7.15 0 5>, rotation=<rot 90 0 0>}]
 dyn_cape_r_03 [{shapeType=Capsule, length=9.602, radius=8.1308, shapeName=dyn_cape_01_capsule, center=<7.15 0 5>, rotation=<rot 90 0 0>}]
 dyn_cape_r_04 [{shapeType=Capsule, length=10, radius=8, shapeName=dyn_cape_01_capsule, center=<10 0 5>, rotation=<rot 90 0 0>}]
 dyn_cape_r_MSTR [{shapeType=Capsule, length=0.505, radius=0.505, shapeName=dyn_cape_r_MSTR_capsule, rotation=<rot 0 -90 0>}]
 dyn_furryball_down_01 [{shapeType=Capsule, length=1, radius=3, shapeName=dyn_furryball_down_01_capsule, center=<2.5 0 0>, rotation=<rot 90 0 0>}]
 dyn_furryball_down_02 [{shapeType=Capsule, length=4, radius=4, shapeName=dyn_furryball_down_02_capsule, center=<5 0 0>, rotation=<rot 90 0 0>}]
 dyn_furryball_middle_01 [{shapeType=Capsule, length=1.5, radius=3, shapeName=dyn_furryball_middle_01_capsule, center=<2.5 0 0>, rotation=<rot 90 0 0>}]
 dyn_furryball_middle_02 [{shapeType=Capsule, length=4, radius=4, shapeName=dyn_furryball_middle_02_capsule, center=<5 0 0>, rotation=<rot 90 0 0>}]
 dyn_furryball_top_01 [{shapeType=Capsule, length=1, radius=3, shapeName=dyn_furryball_top_01_capsule, center=<2.5 0 0>, rotation=<rot 90 0 0>}]
 dyn_furryball_top_02 [{shapeType=Capsule, length=4, radius=4, shapeName=dyn_furryball_top_02_capsule, center=<5 0 0>, rotation=<rot 90 0 0>}]
}
constraints [
 {bone1Name=dyn_cape_01, bone2Name=dyn_cape_MSTR, swing1LimitDegrees=20, swing2LimitDegrees=15, twistLimitDegrees=5}
 {bone1Name=dyn_cape_02, bone2Name=dyn_cape_01, swing1LimitDegrees=10, swing2LimitDegrees=40, twistLimitDegrees=0}
 {bone1Name=dyn_cape_03, bone2Name=dyn_cape_02, swing1LimitDegrees=40, swing2LimitDegrees=45, twistLimitDegrees=0}
 {bone1Name=dyn_cape_04, bone2Name=dyn_cape_03, swing1LimitDegrees=45, swing2LimitDegrees=45, twistLimitDegrees=0}
 {bone1Name=dyn_cape_l_01, bone2Name=dyn_cape_l_MSTR, swing1LimitDegrees=25, swing2LimitDegrees=15, twistLimitDegrees=5}
 {bone1Name=dyn_cape_l_02, bone2Name=dyn_cape_l_01, swing1LimitDegrees=30, swing2LimitDegrees=35, twistLimitDegrees=0}
 {bone1Name=dyn_cape_l_03, bone2Name=dyn_cape_l_02, swing1LimitDegrees=40, swing2LimitDegrees=45, twistLimitDegrees=0}
 {bone1Name=dyn_cape_l_04, bone2Name=dyn_cape_l_03, swing1LimitDegrees=45, swing2LimitDegrees=45, twistLimitDegrees=0}
 {bone1Name=dyn_cape_r_01, bone2Name=dyn_cape_r_MSTR, swing1LimitDegrees=25, swing2LimitDegrees=20, twistLimitDegrees=5}
 {bone1Name=dyn_cape_r_02, bone2Name=dyn_cape_r_01, swing1LimitDegrees=30, swing2LimitDegrees=45, twistLimitDegrees=0}
 {bone1Name=dyn_cape_r_03, bone2Name=dyn_cape_r_02, swing1LimitDegrees=40, swing2LimitDegrees=45, twistLimitDegrees=0}
 {bone1Name=dyn_cape_r_04, bone2Name=dyn_cape_r_03, swing1LimitDegrees=45, swing2LimitDegrees=45, twistLimitDegrees=0}
 {
  bone1Name=base_furryball_down
  bone2Name=dyn_cape_brooch
  swing1LimitDegrees=45
  swing1Motion=Free
  swing2LimitDegrees=45
  swing2Motion=Free
  twistLimitDegrees=45
  twistMotion=Free
 }
 {bone1Name=dyn_furryball_down_01, bone2Name=base_furryball_down, swing1LimitDegrees=10, swing2LimitDegrees=10, twistLimitDegrees=1}
 {bone1Name=dyn_furryball_down_02, bone2Name=dyn_furryball_down_01, swing1LimitDegrees=15, swing2LimitDegrees=15, twistLimitDegrees=1}
 {
  bone1Name=base_furryball_middle
  bone2Name=dyn_cape_brooch
  swing1LimitDegrees=45
  swing1Motion=Free
  swing2LimitDegrees=45
  swing2Motion=Free
  twistLimitDegrees=45
  twistMotion=Free
 }
 {bone1Name=dyn_furryball_middle_01, bone2Name=base_furryball_middle, swing1LimitDegrees=10, swing2LimitDegrees=10, twistLimitDegrees=1}
 {bone1Name=dyn_furryball_middle_02, bone2Name=dyn_furryball_middle_01, swing1LimitDegrees=15, swing2LimitDegrees=15, twistLimitDegrees=1}
 {
  bone1Name=base_furryball_top
  bone2Name=dyn_cape_brooch
  swing1LimitDegrees=45
  swing1Motion=Free
  swing2LimitDegrees=45
  swing2Motion=Free
  twistLimitDegrees=45
  twistMotion=Free
 }
 {bone1Name=dyn_furryball_top_01, bone2Name=base_furryball_top, swing1LimitDegrees=20, swing2LimitDegrees=15, twistLimitDegrees=1}
 {bone1Name=dyn_furryball_top_02, bone2Name=dyn_furryball_top_01, swing1LimitDegrees=25, swing2LimitDegrees=20, twistLimitDegrees=1}
]
