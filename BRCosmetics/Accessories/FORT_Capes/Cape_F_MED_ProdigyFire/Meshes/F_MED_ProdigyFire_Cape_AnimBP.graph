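# Post-process-style cape AnimBP for the ProdigyFire back accessory, copying the attached parent's pose and adding secondary motion.
# Backpack root offsets follow spine bend, then feed gameplay or front-end RigidBody simulation with overridden gravity.
# When rigid body is disabled, skydiving-driven bone rotations and three planar-limited Trail chains layer onto that cached result.
# The event graph derives trail alpha from movement and updates gravity, spine offsets and grind-rail sim-space settings.
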AnimBlueprint /BRCosmetics/Accessories/FORT_Capes/Cape_F_MED_ProdigyFire/Meshes/F_MED_ProdigyFire_Cape_AnimBP
target_skeleton=&/F_MED_ProdigyFire_Cape_Skeleton
graphs [
 AnimGraph {nodes [
   n1 = Get SkydivingAlpha() -> (SkydivingAlpha:Rotator)
   n2 = "Transform (Modify) Bone - Bone: dyn_cape_l_03"(ComponentPose=Local To Component(LocalPose=Use cached pose 'RB_Cache'()), Alpha=1, Translation="0,0,0", Rotation=@n1, Scale="1,1,1") {rotationMode=BMM_Additive, rotationSpace=BCS_BoneSpace, alphaBoolBlend {blendOption=Linear}, boneToModify {boneName=dyn_cape_l_03}}
   n3 = "Transform (Modify) Bone - Bone: dyn_cape_03"(ComponentPose=@n2, Alpha=1, Translation="0,0,0", Rotation=@n1, Scale="1,1,1") {rotationMode=BMM_Additive, rotationSpace=BCS_BoneSpace, alphaBoolBlend {blendOption=Linear}, boneToModify {boneName=dyn_cape_03}}
   n4 = "Transform (Modify) Bone - Bone: dyn_cape_r_03"(ComponentPose=@n3, Alpha=1, Translation="0,0,0", Rotation=@n1, Scale="1,1,1") {rotationMode=BMM_Additive, rotationSpace=BCS_BoneSpace, alphaBoolBlend {blendOption=Linear}, boneToModify {boneName=dyn_cape_r_03}}
   n5 = Get BwdMovement() -> (BwdMovement:double)
   n6 = Trail controller - Bone: dyn_cape_l_07(ComponentPose=@n4, Alpha=@n5) {
    bUsePlanarLimit=true
    chainLength=6
    debugLifeTime=0.1
    maxDeltaTime=0.333
    alphaScaleBiasClamp {bInterpResult=true}
    baseJoint {boneName=backpack_root}
    planarLimits [{drivingBone {boneName=backpack_root}, planeTransform {location=<0 -3 0>, rotation=<rot 0 0 -90>, scale=<1 1 1>}}]
    rotationLimits [
     {limitMax=<180 180 180>}
     {limitMax=<180 180 180>}
     {limitMax=<180 180 180>}
     {limitMax=<180 180 180>}
     {limitMax=<180 180 180>}
     {limitMax=<180 180 180>}
    ]
    rotationOffsets [<0 0 0>, <0 0 0>, <0 0 0>, <0 0 0>, <0 0 0>, <0 0 0>]
    trailBone {boneName=dyn_cape_l_07}
    trailRelaxationSpeed {editorCurveData {keys [<at 0 23 RCIM_Linear>,<at 0.284 19.3799 RCIM_Linear>,<at 0.6554 8.5924 RCIM_Linear>,<at 1 5 RCIM_Linear>]}}
   }
   n7 = Trail controller - Bone: dyn_cape_07(ComponentPose=@n6, Alpha=@n5) {
    bUsePlanarLimit=true
    chainLength=6
    debugLifeTime=0.1
    maxDeltaTime=0.0333
    alphaScaleBiasClamp {bInterpResult=true}
    baseJoint {boneName=backpack_root}
    planarLimits [{drivingBone {boneName=backpack_root}, planeTransform {location=<0 0 0>, rotation=<rot 0 0 -90>, scale=<1 1 1>}}]
    rotationLimits [
     {limitMax=<180 180 180>}
     {limitMax=<180 180 180>}
     {limitMax=<180 90 180>, limitMin=<-180 -90 -180>}
     {limitMax=<180 180 180>}
     {limitMax=<180 180 180>}
     {limitMax=<180 180 180>}
    ]
    rotationOffsets [<0 0 0>, <0 0 0>, <0 0 0>, <0 0 0>, <0 0 0>, <0 0 0>]
    trailBone {boneName=dyn_cape_07}
    trailRelaxationSpeed {editorCurveData {keys [<at 0 23 RCIM_Linear>,<at 0.284 19.3799 RCIM_Linear>,<at 0.6554 8.5924 RCIM_Linear>,<at 1 5 RCIM_Linear>]}}
   }
   n8 = Trail controller - Bone: dyn_cape_r_07(ComponentPose=@n7, Alpha=@n5) {
    bUsePlanarLimit=true
    chainLength=6
    debugLifeTime=0.1
    maxDeltaTime=0.0333
    alphaScaleBiasClamp {bInterpResult=true}
    baseJoint {boneName=backpack_root}
    planarLimits [{drivingBone {boneName=backpack_root}, planeTransform {location=<0 0 0>, rotation=<rot 0 0 -90>, scale=<1 1 1>}}]
    rotationLimits [
     {limitMax=<180 180 180>}
     {limitMax=<180 180 180>}
     {limitMax=<180 90 180>, limitMin=<-180 -90 -180>}
     {limitMax=<180 180 180>}
     {limitMax=<180 180 180>}
     {limitMax=<180 180 180>}
    ]
    rotationOffsets [<0 0 0>, <0 0 0>, <0 0 0>, <0 0 0>, <0 0 0>, <0 0 0>]
    trailBone {boneName=dyn_cape_r_07}
    trailRelaxationSpeed {editorCurveData {keys [<at 0 23 RCIM_Linear>,<at 0.284 19.3799 RCIM_Linear>,<at 0.6554 8.5924 RCIM_Linear>,<at 1 5 RCIM_Linear>]}}
   }
   Output Pose(Result=Blend Poses by bool(BlendPose_0=Use cached pose 'RB_Cache'(), BlendPose_1=Component To Local(ComponentPose=@n8), bActiveValue=Get bRigidBodyEnabled()) {blendType=Linear}) {layerGroup=None}
   n9 = "Transform (Modify) Bone - Bone: backpack_root"(
    ComponentPose=Local To Component(LocalPose=Copy Pose From Mesh() {bUseAttachedParent=true})
    Alpha=1
    Translation="0,1.5,-5"
    Rotation="0,0,0"
    Scale="1,1,1"
   ) {
    rotationMode=BMM_Additive
    rotationSpace=BCS_BoneSpace
    scaleMode=BMM_Additive
    scaleSpace=BCS_BoneSpace
    translationMode=BMM_Additive
    translationSpace=BCS_BoneSpace
    alphaBoolBlend {blendOption=Linear}
    alphaScaleBiasClamp {interpSpeedDecreasing=25, interpSpeedIncreasing=25}
    boneToModify {boneName=backpack_root}
    translation {y=1.5, z=-5}
   }
   n10 = "Transform (Modify) Bone - Bone: backpack_root"(ComponentPose=@n9, Alpha=1, Translation=Get backpack_translation(), Rotation=Get backpack_rotate(), Scale="1,1,1") {
    rotationMode=BMM_Additive
    rotationSpace=BCS_BoneSpace
    scaleMode=BMM_Additive
    scaleSpace=BCS_BoneSpace
    translationMode=BMM_Additive
    translationSpace=BCS_BoneSpace
    alphaBoolBlend {blendOption=Linear}
    alphaScaleBiasClamp {interpSpeedDecreasing=25, interpSpeedIncreasing=25}
    boneToModify {boneName=backpack_root}
    rotation {roll=-5}
    translation {y=2.5, z=-5}
   }
   Save cached pose 'Start_Cache'(Pose=Component To Local(ComponentPose=@n10)) {cacheName=Start_Cache}
   n11 = RigidBody!(
    ComponentPose=Local To Component(LocalPose=Use cached pose 'Start_Cache'())
    Alpha
    OverrideWorldGravity:Vector=Get GravityOverride()
    ExternalForce
   ) -> (Pose) {
    bClampLinearTranslationLimitToRefPose=true
    bOverrideWorldGravity=true
    bUseExternalClothCollision=true
    overridePhysicsAsset=&/F_MED_ProdigyFire_Cape_Physics
    simulationSpace=BaseBoneSpace
    alphaBoolBlend {blendOption=Linear}
    baseBoneRef {boneName=backpack_root}
    componentAppliedLinearAccClamp {z=7500}
    componentLinearAccScale=<0.6 0.5 0.5>
    componentLinearVelScale=<0.5 0.4 0.4>
    overrideWorldGravity {z=-980}
    simSpaceSettings {worldAlpha=0.125}
   }
   n12 = RigidBody(ComponentPose=Local To Component(LocalPose=Use cached pose 'Start_Cache'()), OverrideWorldGravity=Get GravityOverride()) {
    bClampLinearTranslationLimitToRefPose=true
    bOverrideWorldGravity=true
    bUseExternalClothCollision=true
    overridePhysicsAsset=&/F_MED_ProdigyFire_Cape_Physics_FrontEnd
    simulationSpace=BaseBoneSpace
    alphaBoolBlend {blendOption=Linear}
    baseBoneRef {boneName=backpack_root}
    componentAppliedLinearAccClamp {z=7500}
    componentLinearAccScale=<0.6 0.5 0.5>
    componentLinearVelScale=<0.5 0.4 0.4>
    overrideWorldGravity {z=-980}
    simSpaceSettings {worldAlpha=0.125}
   }
   Save cached pose 'RB_Cache'(Pose=Blend Poses by bool(BlendPose_0=Component To Local(ComponentPose=@n12), BlendPose_1=Component To Local(ComponentPose=@n11), bActiveValue=Get bIsInFrontEnd()) {blendType=Linear}) {cacheName=RB_Cache}
  ]}
 EventGraph {nodes [
   Event BlueprintUpdateAnimation()
   > n1 = Sequence!(execute) -> (then_0, then_1, then_2:Exec, then_3:Exec)
   n2 = Branch(execute=@n1.then_3, Condition=Get bIsInFrontEnd())
   > Set BwdMovement(BwdMovement:double=0.5) -> (Output_Get:double)
   n3 = PlayerMovement(execute=@n2.else, AnimInstance=Self-Reference(), XDotProductCoefficient=1, YDotProductCoefficient=1, ZDotProductCoefficient=1)
   > n4 = Branch(Condition=Get bIsDBNOCarried())
   > Set BwdMovement(BwdMovement=0.5)
   n5 = Branch(execute=@n4.else, Condition=AND Boolean(A=Get bIsTargeting(), B=Get bIsCrouching()))
   n6 = MapRangeClamped(
    Value=float + float(A="Absolute (Float)"(A=@n3), B=MapRangeClamped(Value=@n3.YMovement, InRangeA=-1, InRangeB=1, OutRangeA=0.3)) {is=title K2Node_CommutativeAssociativeBinaryOperator}
    InRangeA=0
    InRangeB=1.65
   )
   > Set BwdMovement(BwdMovement=float * float(A=@n6, B=0.25) {is=title K2Node_PromotableOperator})
   Set BwdMovement(execute=@n5.else, BwdMovement=@n6)
   n7 = Cast To FortPlayerPawn(execute=@n1.then_1, Object=TryGetPawnOwner())
   > n8 = Branch(Condition=Get bIsOnGrindRail())
   > n9 = Branch(Condition=NOT Boolean(A=IsWeaponHolstered!(self:Fort Pawn Object Reference=@n7) -> (ReturnValue:bool)))
   > n10 = Branch(Condition=Get bIsTargeting())
   > Set Sim Space Settings(Sim Space Settings:Sim Space Settings Structure=Get Sim Space Settings_GrindTarget() -> (Sim Space Settings_GrindTarget:Sim Space Settings Structure)) -> (Output_Get:Sim Space Settings Structure)
   Set Sim Space Settings(
    execute=@n10.else
    Sim Space Settings=Get Sim Space Settings_GrindingWeaponActive() -> (Sim Space Settings_GrindingWeaponActive:Sim Space Settings Structure)
   )
   Set Sim Space Settings(execute=[@n8.else, @n9.else], Sim Space Settings=Get Sim Space Settings_Base() -> (Sim Space Settings_Base:Sim Space Settings Structure))
   n11 = Gravity Override State Function DEPRECATED(
    execute=@n1.then_0
    CharacterPartAnimInstance=Self-Reference()
    GameTimeInSeconds=GetGameTimeInSeconds()
    GravityOverrideParams=Get Gravity Override Params()
    DebugJointName=spine_05
   )
   > Set GravityOverride(GravityOverride=@n11)
   n12 = GetOwnerSkeletalMeshComponent()
   n13 = BreakRotator(InRot=BreakTransform(InTransform=GetDeltaTransformFromRefPose(self=@n12, BoneName=spine_05, BaseName=spine_01)).Rotation)
   Set backpack_rotate(execute=@n1.then_2, backpack_rotate=MakeRotator(Roll=MapRangeClamped(Value=@n13.Yaw, InRangeA=0, InRangeB=-30, OutRangeB=35)))
   > Set backpack_translation(backpack_translation=MakeVector(X=2.5, Y=MapRangeClamped(Value=@n13.Yaw, InRangeA=0, InRangeB=-40, OutRangeB=-10), Z=-1.5))
   Event OnBeginSkydiving()
   > Set SkydivingAlpha(SkydivingAlpha:Rotator="65,0,0") -> (Output_Get:Rotator)
   Event OnEndSkydiving()
   > Set SkydivingAlpha(SkydivingAlpha="10,0,0")
   n14 = Branch(Condition=Get bIsSkyDiving())
   > n15 = Branch(Condition="float > float"(A=Get PawnVelocityZ(), B=50) {is=title K2Node_CallFunction})
   > Set BwdMovement(BwdMovement=float * float(A=Get BwdMovement(), B=0) {is=title K2Node_CommutativeAssociativeBinaryOperator})
   Set BwdMovement(execute=[@n14.else, @n15.else], BwdMovement=float * float(A=Get BwdMovement(), B=1))
   TryGetPawnOwner()
  ]}
]
variables [
 backpack_rotate {type=Rotator}
 backpack_translation {type=Vector}
 BwdMovement {type=number}
 Cape_LookAt_Target {type=Vector}
 Gravity Override Params {type=GravityOverrideParamsStruct}
 GravityOverride {type=Vector}
 Sim Space Settings {
  type=SimSpaceSettings
  default {
   dampingAlpha=1
   maxAngularAcceleration=10000
   maxAngularVelocity=10000
   maxLinearAcceleration=10000
   maxLinearVelocity=10000
   velocityScaleZ=1
   worldAlpha=0
   externalLinearDragV=<0 0 0>
  }
 }
 Sim Space Settings_Base {
  type=SimSpaceSettings
  default {
   dampingAlpha=1
   maxAngularAcceleration=10000
   maxAngularVelocity=10000
   maxLinearAcceleration=10000
   maxLinearVelocity=10000
   velocityScaleZ=1
   worldAlpha=0
   externalLinearDragV=<0 0 0>
  }
 }
 Sim Space Settings_GrindingWeaponActive {
  type=SimSpaceSettings
  default {
   dampingAlpha=1
   maxAngularAcceleration=10000
   maxAngularVelocity=10000
   maxLinearAcceleration=10000
   maxLinearVelocity=10000
   velocityScaleZ=1
   worldAlpha=0
   externalLinearDragV=<0 0 0>
  }
 }
 Sim Space Settings_GrindTarget {
  type=SimSpaceSettings
  default {
   dampingAlpha=1
   maxAngularAcceleration=10000
   maxAngularVelocity=10000
   maxLinearAcceleration=10000
   maxLinearVelocity=10000
   velocityScaleZ=1
   worldAlpha=0
   externalLinearDragV=<0 0 0>
  }
 }
 SkydivingAlpha {type=Rotator}
]
